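# Master skeletal-mesh material for the Pliant female character body, adding a parameterized primary-color tint to skin.
# It runs shared PreFX and CharacterSkinFX functions, then tints base color through desaturation and power, and emissive through color and multiplier.
# The SkinFX_Mask green channel blends both tints over the original attributes before the shared PostFX function.
# The emissive multiplier has a separate mobile parameter, while the emissive desaturate-power branch is unconnected.
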Material /Game/Characters/Player/Female/Medium/Bodies/F_MED_Pliant/Meshes/Materials/M_Pliant_Body_Master
graph {
 properties {
  bUseMaterialAttributes=true
  bUsedWithMeshDeformer=true
  bUsedWithMorphTargets=true
  bUsedWithSkeletalMesh=true
  pixelDepthOffsetMode=PDOM_Legacy
  refractionMethod=RM_IndexOfRefraction
  shadingModel=MSM_FromMaterialExpression
  usedShadingModels=Default Lit | Subsurface | Clear Coat | Subsurface Profile | Cloth | Eye
 }
 nodes [
  n1 = MaterialFunctionCall() {materialFunction=&BaseFN_CharacterFunction_PreFX}
  n2 = GetMaterialAttributes(None=MaterialFunctionCall(
    PreskinnedPosition=@n1.Pre skinned Local Position
    Gmap=@n1.Pre skinned Local Position
    MaterialAttributes=@n1.Pre skinned Local Position
    AO=@n1.Pre skinned Local Position
   ) {materialFunction=&CharacterSkinFX}.Result) {attributeGetTypes [#1, #2]}
  n3 = VectorParameter() {group=Pliant, parameterName=Pliant_PrimaryColor, defaultValue=<rgba 0.9271 0.0922 0.5072 1>}
  n4 = ScalarParameter() {defaultValue=1, group=Pliant, parameterName=Pliant_BodyTintEmsMult}
  n5 = Multiply(
   A=Multiply(A=@n2.Emissive Color, B=@n3.RGB)
   B=ShadingPathSwitch(Default=@n4, Deferred=@n4, Mobile=ScalarParameter() {defaultValue=1, group=Pliant, parameterName=Pliant_BodyTintEmsMult_Mobile}) {desc=Migrated from FeatureLevelSwitch}
  )
  n6 = TextureSampleParameter2D() {
   parameterName=SkinFX_Mask
   samplerType=SAMPLERTYPE_LinearColor
   texture=/Game/Characters/Player/Female/Medium/Bodies/F_MED_Pliant/Textures/T_F_MED_Pliant_Body_E
  }
  n7 = Multiply(
   A=@n3.RGB
   B=Power(
    Base=Desaturation(None=@n2.Base Color)
    Exp=ScalarParameter() {defaultValue=1, group=Pliant, parameterName=Pliant_BodyDiffuseTintPow, sliderMax=2}
   )
  )
  MaterialFunctionCall(
   AO=@n1.AO
   Distance Blend=@n1.AO
   Fresnel=@n1.AO
   Gmap=@n1.AO
   Pre Skinned Local Position=@n1.AO
   Material=SetMaterialAttributes!(
    MaterialAttributes=@n2.MaterialAttributes
    Emissive Color=LinearInterpolate(A=@n2.Emissive Color, B=@n5, Alpha=@n6.G)
    Base Color=LinearInterpolate(A=@n2.Base Color, B=@n7, Alpha=@n6.G)
   ) {attributeSetTypes [#1, #2]}
  ) {materialFunction=&BaseFN_CharacterFunction_PostFX}
  Power(
   Base=Desaturation(None=@n2.Emissive Color)
   Exp=ScalarParameter() {defaultValue=1, group=Pliant, parameterName=Pliant_BodyEmsTintPow, sliderMax=2}
  )
 ]
}
